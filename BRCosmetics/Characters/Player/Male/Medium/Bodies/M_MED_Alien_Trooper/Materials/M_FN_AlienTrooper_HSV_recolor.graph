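# Material function for the Alien Trooper cosmetic body that recolors an input color by hue, saturation and value.
# It hue-shifts inColor, desaturates the result by one minus the Saturation input, then multiplies it by Value.
# Saturation can only reduce color toward gray within a 0–1 range, and Value acts as a plain brightness multiplier rather than a true HSV conversion.
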MaterialFunction /BRCosmetics/Characters/Player/Male/Medium/Bodies/M_MED_Alien_Trooper/Materials/M_FN_AlienTrooper_HSV_recolor
graph {nodes [
  n1 = MaterialFunctionCall(Hue Shift Percentage=FunctionInput() {inputName=HueShift, inputType=FunctionInput_Scalar}, Texture=FunctionInput() {inputName=inColor}) {materialFunction=&HueShift}
  n2 = OneMinus(None=FunctionInput() {inputName=Saturation, inputType=FunctionInput_Scalar})
  FunctionOutput(None=Multiply(A=Desaturation(None=@n1.Result, Fraction=@n2), B=FunctionInput() {inputName=Value, inputType=FunctionInput_Scalar}))
 ]}
