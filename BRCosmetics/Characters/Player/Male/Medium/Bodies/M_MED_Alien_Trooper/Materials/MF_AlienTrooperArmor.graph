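# Material function that recolors the Alien Trooper player body armor and adds a customizable emblem.
# FX mask channels drive secondary color, accent color and emissive glow, while an inverted combined mask limits the HSV-shifted primary color.
# The emblem samples an offset, center-scaled second UV set, blending between emblem and banner textures by parameter.
# A metallic emblem-colored layer is blended into the armor attributes, then passed through the shared weapon-skin Clearcoat function.
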MaterialFunction /BRCosmetics/Characters/Player/Male/Medium/Bodies/M_MED_Alien_Trooper/Materials/MF_AlienTrooperArmor
graph {nodes [
  Constant() {r=1}
  FunctionInput() {inputName=Skin, inputType=FunctionInput_Scalar}
  n1 = NamedRerouteUsage()
  n2 = GetMaterialAttributes(None=FunctionInput() {inputName=Material, inputType=FunctionInput_MaterialAttributes}) {attributeGetTypes [#1]}
  n3 = VectorParameter() {group=AlienTrooper, parameterName=AT_SuitPrimary_HSV, sortPriority=20, defaultValue=<rgba 1 1 1 0>}
  n4 = MaterialFunctionCall!(HueShift=@n3.R, Saturation=@n3.R, Value=@n3.R, inColor=@n2.BaseColor) {materialFunction=&/M_FN_AlienTrooper_HSV_recolor}
  n5 = TextureSampleParameter2D() {group=AlienTrooper, parameterName=M, sortPriority=5, texture=&T_BlacksRGB}
  n6 = TextureSampleParameter2D() {group=AlienTrooper, parameterName=AT_FXMask, sortPriority=5, texture=&T_BlacksRGB}
  n7 = Add(A=Add(A=@n6.R, B=@n6.R), B=@n6.B)
  n8 = OneMinus(None=Saturate(None=Add(A=@n5.B, B=@n7)))
  n9 = LinearInterpolate(
   A=LinearInterpolate(
    A=LinearInterpolate(A=@n2.BaseColor, B=@n4.Result, Alpha=@n8)
    B=VectorParameter() {group=AlienTrooper, parameterName=AT_SuitSecondary, sortPriority=20, defaultValue=<rgba 1 1 1 0>}.RGB
    Alpha=@n6.R
   )
   B=VectorParameter() {group=AlienTrooper, parameterName=AT_SuitAccent, sortPriority=20, defaultValue=<rgba 1 1 1 0>}.RGB
   Alpha=@n6.G
  )
  n10 = VectorParameter() {group=AlienTrooper, parameterName=AT_EmblemColor, sortPriority=50, defaultValue=<rgba 0.05 0.07 0.08 1>}
  n11 = MaterialFunctionCall(UVs=Add(A=TextureCoordinate() {coordinateIndex=1}, B=Constant2Vector() {g=-0.05, r=0.02}), Texture Scale=Constant() {r=0.35}) {materialFunction=&ScaleUVsByCenter}
  n12 = TextureSampleParameter2D(UVs=@n11.UVs) {group=AlienTrooper, parameterName=AT_Emblem_Texture, sortPriority=50, texture=&T_BlacksRGB}
  n13 = TextureSampleParameter2D(UVs=@n11.UVs) {
   group=AlienTrooper
   parameterName=Banner_Texture
   sortPriority=50
   texture=/Game/UI/Foundation/Textures/Banner/Standard/T-Banners-Icons-Waves-L
  }
  n14 = Multiply(
   A=LinearInterpolate(A=@n12.RGB, B=@n13.RGB, Alpha=ScalarParameter() {group=AlienTrooper, parameterName=AT_UseBanner, sortPriority=50})
   B=@n11.0-1 mask
  )
  n15 = LinearInterpolate(A=LinearInterpolate(A=Multiply(A=@n9, B=@n5.R), B=@n9), B=@n10.RGB, Alpha=@n14) {constB=0.01}
  n16 = Multiply(A=VectorParameter() {group=AlienTrooper, parameterName=AT_SuitEmissive, sortPriority=20}.RGB, B=Power(Base=@n6.B))
  n17 = NamedRerouteUsage()
  n18 = Constant() {r=1}
  n19 = MaterialFunctionCall!(
   Alpha=NamedRerouteUsage()
   In=BlendMaterialAttributes(
    A=SetMaterialAttributes!(MaterialAttributes=@n1, BaseColor=@n15, EmissiveColor=@n16, ShadingModel) {attributeSetTypes [#1, #2, #3]}
    B=SetMaterialAttributes!(MaterialAttributes=@n17, BaseColor=@n10.RGB, Metallic=@n18) {attributeSetTypes [#1, #4]}
    Alpha=NamedRerouteUsage()
   )
  ) {materialFunction=/Game/Weapons/WeaponSkins/MaterialFunction/Clearcoat}
  FunctionOutput(None=@n19.Result)
  NamedRerouteDeclaration(None=@n2.MaterialAttributes) {name=MaterialAttributes, nodeColor=<rgba 0 1 0.3672 1>}
  NamedRerouteDeclaration(None=ComponentMask(None=@n14) {r=true}) {name=EmblemAlpha, nodeColor=<rgba 0.4062 1 0 1>}
  NamedRerouteDeclaration(None=@n8) {name=primaryMask, nodeColor=<rgba 0 1 0.9297 1>}
  SetMaterialAttributes!(
   MaterialAttributes=@n19.Result
   BaseColor=Multiply(A=TextureSampleParameter2D() {group=AlienTrooper, parameterName=AT_FXMask, sortPriority=5, texture=&T_BlacksRGB}.RGB) {constB=100}
  ) {attributeSetTypes [#1]}
  VectorParameter() {group=AlienTrooper, parameterName=Banner_PrimaryColor, sortPriority=50, defaultValue=<rgba 0 1 1 0>}
  VertexColor()
 ]}
