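# Material function that samples the Alien Trooper's eye color and mask textures with simulated refraction, defaulting to the Cyclops eye textures.
# ML_AlienTrooperEyeRefraction derives refracted UVs from sphere-mask depth inputs and a center-flattened eye normal, blended with plain center-scaled UVs by "Refraction On/Off".
# The result is rescaled by iris radius, radially warped by custom HLSL for PupilScale, then inverted before sampling both overridable texture inputs.
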MaterialFunction /BRCosmetics/Characters/Player/Male/Medium/Bodies/M_MED_Alien_Trooper/Materials/MF_alienTrooperRefractiveEye
graph {nodes [
  Constant()
  n1 = Constant() {r=0.5}
  n2 = ScalarParameter() {defaultValue=1, parameterName=ScaleByCenter}
  n3 = ScalarParameter() {defaultValue=0.035, desc=Color transition range width, parameterName=LimbusUVWidthColor}
  n4 = ScalarParameter() {defaultValue=0.045, desc=Shading transition range width, parameterName=LimbusUVWidthShading}
  n5 = ScalarParameter() {defaultValue=0.25, parameterName=Iris UV Radius, sliderMax=0.5}
  n6 = Multiply(A=@n5, B=@n2)
  n7 = SphereMask(
   A=MaterialFunctionCall(X=Add(A=@n6) {constB=0.5}, Y=Constant() {r=0.5}) {materialFunction=&MakeFloat2}.Result
   B=Constant2Vector() {g=0.5, r=0.5}
  ) {attenuationRadius=0.36, hardnessPercent=0}
  n8 = TextureCoordinate()
  n9 = Constant2Vector() {g=0.5, r=0.5}
  n10 = Transform(None=MaterialFunctionCall(
    Normal=TextureSample() {samplerType=SAMPLERTYPE_Normal, texture=/Game/Characters/Shared/EyeMaterials/Textures/T_Eye_N}.RGB
    Flatness=SphereMask(A=@n8, B=@n9) {attenuationRadius=0.1, hardnessPercent=0}
   ) {materialFunction=&FlattenNormal}.Result)
  n11 = SphereMask(A=@n8, B=@n9) {attenuationRadius=0.36, hardnessPercent=0}
  n12 = LinearInterpolate(
   A=MaterialFunctionCall(Texture Scale=@n2) {materialFunction=&ScaleUVsByCenter}.UVs
   B=MaterialFunctionCall!(
    DepthScale=ScalarParameter() {defaultValue=1, parameterName=DepthScale}
    InternalIoR=ScalarParameter() {defaultValue=1.336, desc=Internal IoR, parameterName=IoR, sliderMax=10, sliderMin=1.0003}
    ScaleByCenter=@n2
    LimbusUVWidth=MaterialFunctionCall(X=@n3, Y=@n4) {materialFunction=&MakeFloat2}.Result
    IrisUVRadius=@n5
    DepthPlaneOffset=Multiply(A=@n7) {constB=4}
    MidPlaneDisplacement=Multiply(A=@n11) {constB=2.2}
    EyeDirectionWorld=Normalize(VectorInput=@n10)
    UVS
   ) {materialFunction=&/ML_AlienTrooperEyeRefraction}.RefractedUV
   Alpha=ScalarParameter() {defaultValue=0.25, parameterName=Refraction On/Off, sliderMax=1}
  ) {constA=1, desc=UVs with Refraction or not}
  n13 = Constant() {r=1}
  n14 = Multiply(A=@n5, B=Constant() {r=2}) {desc=Radius to diameter}
  MaterialExpressionAdd_1 = Add(
   A=Multiply(A=Subtract(A=@n12, B=@n1), B=Divide(A=@n13, B=@n14) {desc="Scale Iris texture coordinates up\r\nby this amount before sampling iris"}) {desc=Centered Scale}
   B=@n1
  )
  MaterialExpressionScalarParameter_25 = ScalarParameter() {defaultValue=1, parameterName=PupilScale, sliderMax=1}
  n15 = Custom!(UV=@MaterialExpressionAdd_1, PupilScale=@MaterialExpressionScalarParameter_25) {
   code="// Scale UVs from from unit circle in or out from center\r\n// float2 UV, float PupilScale\r\n\r\nfloat2 UVcentered = UV - float2(0.5f, 0.5f);\r\nfloat UVlength = length(UVcentered);\r\n// UV on circle at distance 0.5 from the center, in direction of original UV\r\nfloat2 UVmax = normalize(UVcentered)*0.5f;\r\n\r\nfloat2 UVscaled = lerp(UVmax, float2(0.f, 0.f), saturate((1.f - UVlength*2.f)*PupilScale));\r\nreturn UVscaled + float2(0.5f, 0.5f);"
   desc=Scale Pupils
   outputType=CMOT_Float2
   inputs [
    {
     inputName=UV
     input {expression=&self:MaterialExpressionAdd_1}
    }
    {
     inputName=PupilScale
     input {expression=&self:MaterialExpressionScalarParameter_25}
    }
   ]
  }
  n16 = OneMinus(None=@n15)
  n17 = TextureObject() {texture=/BRCosmetics/Characters/Player/Male/Medium/Bodies/M_MED_Alien_Trooper/Textures/Cyclops/T_M_MED_AlienTrooper_Cyclops_Eye_D}
  FunctionOutput(None=TextureSample(UVs=@n16, Tex=FunctionInput(Preview=@n17) {inputName=TextureIn_color, inputType=FunctionInput_Texture2D}).RGB)
  n18 = TextureObject() {
   samplerType=SAMPLERTYPE_Masks
   texture=/BRCosmetics/Characters/Player/Male/Medium/Bodies/M_MED_Alien_Trooper/Textures/Cyclops/T_M_MED_AlienTrooper_Cyclops_Eye_FX
  }
  FunctionOutput(None=TextureSample(UVs=@n16, Tex=FunctionInput(Preview=@n18) {inputName=TextureIn_masks, inputType=FunctionInput_Texture2D}) {samplerType=SAMPLERTYPE_Masks}.RGB) {outputName=Masks}
 ]}
