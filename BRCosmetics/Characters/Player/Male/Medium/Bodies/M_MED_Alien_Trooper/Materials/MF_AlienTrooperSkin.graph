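# Material function that skins the customizable Alien Trooper player body, overriding base color, normal, emissive, roughness and specular on input material attributes.
# HSV parameters recolor the diffuse, with FX-mask channels supplying secondary, pattern and emissive masks.
# Vertex-color sphere masks isolate membrane, teeth and eyes, while pre-skinned world-aligned noise provides a 3D noise source.
# Static switches enable a blended eye material with fresnel color, pupil and refractive cyclops variant.
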MaterialFunction /BRCosmetics/Characters/Player/Male/Medium/Bodies/M_MED_Alien_Trooper/Materials/MF_AlienTrooperSkin
graph {nodes [
  FunctionInput() {inputName=Skin, inputType=FunctionInput_Scalar}
  n1 = Subtract(A=ComponentMask(None=NamedRerouteUsage()) {g=true, r=true}) {constB=0.4}
  n2 = MaterialFunctionCall(Float3=VectorParameter() {group=AlienTrooper, parameterName=AT_Primary_HSV, sortPriority=20, defaultValue=<rgba 0 1 1 0>}.RGB) {materialFunction=&BreakOutFloat3Components}
  n3 = TextureSampleParameter2D() {
   group=AlienTrooper
   parameterName=AT_Diffuse
   sortPriority=10
   texture=/BRCosmetics/Characters/Player/Male/Medium/Bodies/M_MED_Alien_Trooper/Textures/Horns/T_M_MED_AlienTrooper_Horns_Head_D
  }
  n4 = NamedRerouteDeclaration(None=@n3.RGB) {name=baseColor, nodeColor=<rgba 0 0.9531 1 1>}
  n5 = LinearInterpolate(A=@n4, B=Multiply(A=@n4, B=Constant3Vector() {constant=<rgba 3.5 2.6856 2.1 1>}), Alpha=NamedRerouteUsage())
  n6 = MaterialFunctionCall!(HueShift=Add(A=Multiply(A=@n1) {constB=0.1}, B=@n2.R), Saturation=@n2.G, Value=@n2.G, inColor=@n5) {materialFunction=&/M_FN_AlienTrooper_HSV_recolor}
  n7 = Multiply(A=ComponentMask(None=@n1) {r=true}) {constB=0.25}
  n8 = VectorParameter() {group=AlienTrooper, parameterName=AT_Secondary_HSV, sortPriority=30, defaultValue=<rgba 0 1 1 0>}
  n9 = LinearInterpolate(
   A=LinearInterpolate(A=@n6.Result, B=@n5, Alpha=NamedRerouteUsage())
   B=MaterialFunctionCall(HueShift=Add(A=@n7, B=@n8.R), Saturation=@n8.G, Value=Add(A=@n7, B=@n8.B), inColor=@n5) {materialFunction=&M_FN_AlienTrooper_HSV_recolor}.Result
   Alpha=NamedRerouteUsage()
  )
  n10 = VectorParameter() {group=AlienTrooper, parameterName=AT_PatternColor, sortPriority=40, defaultValue=<rgba 0.03 0.8 0.03 1>}
  n11 = NamedRerouteUsage()
  n12 = NamedRerouteUsage()
  n13 = GetMaterialAttributes(None=NamedRerouteDeclaration(None=FunctionInput() {inputName=Material, inputType=FunctionInput_MaterialAttributes}) {name=InputMaterial, nodeColor=<rgba 1 0.7969 0 1>}) {attributeGetTypes [#1]}
  n14 = Multiply(A=@n13.Roughness) {constB=1.4}
  n15 = NamedRerouteUsage()
  n16 = SetMaterialAttributes!(
   MaterialAttributes=@n13.MaterialAttributes
   BaseColor=LinearInterpolate(A=LinearInterpolate(A=@n9, B=@n10.RGB, Alpha=@n11), B=Constant3Vector(), Alpha=Multiply(A=@n12) {constB=0.8})
   Normal=NamedRerouteUsage()
   EmissiveColor=NamedRerouteUsage()
   Roughness=LinearInterpolate(A=LinearInterpolate(A=@n13.Roughness, B=@n14, Alpha=@n15), B=Multiply(A=@n13.Roughness) {constB=1.2}, Alpha=NamedRerouteUsage())
   Specular
  ) {attributeSetTypes [
    #2
    #3
    #4
    #1
    #5
   ]}
  n17 = BlendMaterialAttributes(A=@n16, B=NamedRerouteUsage(), Alpha=NamedRerouteUsage())
  FunctionOutput(None=SetMaterialAttributes!(MaterialAttributes=StaticSwitchParameter(True=@n17, False=@n16) {group=AlienTrooper, parameterName=AT_UseEyes}, BaseColor) {attributeSetTypes [#2]})
  n18 = TextureSampleParameter2D() {
   group=AlienTrooper
   parameterName=AT_FXMask
   samplerType=SAMPLERTYPE_Masks
   sortPriority=5
   texture=&T_BlackMask
  }
  n19 = MaterialFunctionCall(Color=StaticSwitchParameter(True=NamedRerouteUsage(), False=@n18.RGB) {parameterName=useRefractiveEye}) {materialFunction=&SplitComponents}
  NamedRerouteDeclaration(None=NamedRerouteDeclaration(None=@n19.R) {name=SecondaryMask, nodeColor=<rgba 0.2188 0 1 1>}) {name=PupilMask, nodeColor=<rgba 0.8281 0 1 1>}
  NamedRerouteDeclaration(None=NamedRerouteDeclaration(None=@n19.G) {name=PatternMask, nodeColor=<rgba 1 0.2109 0 1>}) {name=IrisGradient, nodeColor=<rgba 0.9688 1 0 1>}
  n20 = TextureSampleParameter2D() {
   group=AlienTrooper
   parameterName=Normals
   samplerType=SAMPLERTYPE_Normal
   sortPriority=100
   texture=/BRCosmetics/Characters/Player/Male/Medium/Bodies/M_MED_Alien_Trooper/Textures/Horns/T_M_MED_AlienTrooper_Horns_Head_N
  }
  NamedRerouteDeclaration(None=@n20.RGB) {name=Normals, nodeColor=<rgba 0 0.5078 1 1>}
  n21 = VertexColor()
  n22 = Constant() {r=0.3}
  NamedRerouteDeclaration(None=Multiply(A=SphereMask(A=@n21.R, B=@n22) {attenuationRadius=0.05}, B=ScalarParameter() {group=AlienTrooper, parameterName=AT_MembraneMult})) {name=MembraneMask, nodeColor=<rgba 0 1 0.1562 1>}
  n23 = NamedRerouteDeclaration(None=@n19.B) {name=EmissiveMask, nodeColor=<rgba 0 1 0.1797 1>}
  NamedRerouteDeclaration(None=Multiply(A=Power(Base=@n23) {constExponent=1}) {constB=0.5}) {name=IrisFleks, nodeColor=<rgba 0.2188 0 1 1>}
  NamedRerouteDeclaration(None=@n19.RGB) {name=masks, nodeColor=<rgba 1 0.0234 0 1>}
  n24 = VectorParameter() {group=AlienTrooper, parameterName=AT_EyePrimary_HSV, sortPriority=20, defaultValue=<rgba 0 1 1 0>}
  n25 = NamedRerouteUsage()
  n26 = Multiply(A=@n25, B=ScalarParameter() {defaultValue=0.15, parameterName=AT_irisFlekBoost}) {constB=0.15}
  n27 = Multiply(A=@n25, B=ScalarParameter() {parameterName=AT_irisFlekBrightness}) {constB=0.15}
  n28 = StaticSwitchParameter(
   True=MaterialFunctionCall(HueShift=@n24.R, Saturation=@n24.R, Value=@n24.R, inColor=NamedRerouteUsage()) {materialFunction=&M_FN_AlienTrooper_HSV_recolor}.Result
   False=MaterialFunctionCall(HueShift=Add(A=@n24.R, B=@n26), Saturation=@n24.G, Value=Add(A=@n24.B, B=@n27), inColor=NamedRerouteUsage()) {materialFunction=&M_FN_AlienTrooper_HSV_recolor}.Result
  ) {parameterName=useRefractiveEye}
  n29 = VectorParameter() {group=AlienTrooper, parameterName=AT_EyeColor, sortPriority=100, defaultValue=<rgba 0.6 0.6 0.6 1>}
  n30 = VectorParameter() {group=AlienTrooper, parameterName=AT_EyeColorFresnel, sortPriority=100, defaultValue=<rgba 0.6 0.6 0.6 1>}
  n31 = PixelNormalWS()
  n32 = LinearInterpolate(
   A=@n29.RGB
   B=@n30.RGB
   Alpha=Power(
    Base=Fresnel(Normal=@n31) {baseReflectFraction=0, exponent=2}
    Exp=ScalarParameter() {defaultValue=0.25, parameterName=AT_EyeColorFresnelExp}
   ) {constExponent=0.25}
  )
  n33 = ScalarParameter() {parameterName=AT_overrideEyeColor}
  n34 = NamedRerouteUsage()
  n35 = ScalarParameter() {defaultValue=1, group=AlienTrooper, parameterName=AT_showPupil}
  NamedRerouteDeclaration(None=LinearInterpolate(
    A=LinearInterpolate(A=@n28, B=@n32, Alpha=@n33)
    B=VectorParameter() {parameterName=AT_PupilColor, defaultValue=<rgba 0.01 0.01 0.01 1>}.RGB
    Alpha=Multiply(A=@n34, B=@n35)
   )) {name=EyeBaseColor, nodeColor=<rgba 1 0.1172 0 1>}
  n36 = TextureObject() {
   samplerType=SAMPLERTYPE_Masks
   texture=/BRCosmetics/Characters/Player/Male/Medium/Bodies/M_MED_Alien_Trooper/Textures/Cyclops/T_M_MED_AlienTrooper_Cyclops_Eye_FX
  }
  n37 = MaterialFunctionCall!(
   TextureIn_color=TextureObject() {texture=/BRCosmetics/Characters/Player/Male/Medium/Bodies/M_MED_Alien_Trooper/Textures/Cyclops/T_M_MED_AlienTrooper_Cyclops_Eye_D}
   TextureIn_masks=@n36
  ) {materialFunction=&/MF_alienTrooperRefractiveEye}
  NamedRerouteDeclaration(None=@n37.Masks) {name=CyclopsMasks, nodeColor=<rgba 0 1 0.5781 1>}
  n38 = TextureObjectParameter() {
   group=SuperLevelling_Mask
   parameterName=SL_TransitionNoise
   samplerType=SAMPLERTYPE_Masks
   texture=/Game/Global/Textures/Default/Noise/NoisePattern
  }
  n39 = PreSkinnedPosition()
  n40 = PreSkinnedNormal()
  NamedRerouteDeclaration(None=MaterialFunctionCall(
    TextureObject=@n38
    TextureSize=Constant3Vector() {constant=<rgba 80 50 50 0>}
    WorldPosition=VertexInterpolator(VS=@n39).PS
    World Space Normal=VertexInterpolator(VS=@n40).PS
   ) {materialFunction=/Game/Weapons/WeaponSkins/Wraps/MaterialFunctions/WorldAlignedTexture}.XYZ Texture) {name=3dNoise, nodeColor=<rgba 0 1 0.5547 1>}
  NamedRerouteDeclaration(None=@n37.Result) {name=CyclopsColor, nodeColor=<rgba 1 0.0234 0 1>}
  n41 = NamedRerouteUsage()
  n42 = Fresnel(Normal=PixelNormalWS()) {baseReflectFraction=0.2, exponent=1}
  NamedRerouteDeclaration(None=Multiply(A=Multiply(A=@n41, B=@n42)) {constB=4}) {name=EmissiveMaskFresnel, nodeColor=<rgba 1 0 0.5391 1>}
  n43 = NamedRerouteUsage()
  n44 = Multiply(A=Subtract(A=ComponentMask(None=@n43) {g=true}) {constB=0.5}) {constB=0.1}
  n45 = VectorParameter() {group=AlienTrooper, parameterName=AT_EmissiveColor, sortPriority=50, defaultValue=<rgba 1 1 1 1>}
  NamedRerouteDeclaration(None=Multiply(
    A=MaterialFunctionCall(Hue Shift Percentage=@n44, Texture=@n45.RGB) {materialFunction=&HueShift}.Result
    B=NamedRerouteUsage()
   )) {name=Emissive, nodeColor=<rgba 1 0 0.9844 1>}
  NamedRerouteDeclaration(None=SphereMask(A=VertexColor(), B=Constant()) {attenuationRadius=0.1}) {name=EyeMask, nodeColor=<rgba 0.3125 0 1 1>}
  n46 = VertexColor()
  n47 = Constant() {r=0.3}
  n48 = ScalarParameter() {group=AlienTrooper, parameterName=AT_MembraneMult}
  NamedRerouteDeclaration(None=Multiply(A=SphereMask(A=@n46.R, B=@n47) {attenuationRadius=0.02}, B=OneMinus(None=@n48))) {name=TeethMask, nodeColor=<rgba 0.7344 1 0 1>}
  n49 = SetMaterialAttributes!(
   MaterialAttributes=NamedRerouteUsage()
   BaseColor=NamedRerouteUsage()
   Normal=NamedRerouteUsage()
   EmissiveColor
   Specular=Multiply(A=Constant() {r=0.3}, B=ScalarParameter() {defaultValue=1, group=AlienTrooper, parameterName=AT_EyeSpecMult, sortPriority=100})
   Roughness=Constant() {r=0.15}
   Metallic=ScalarParameter() {group=AlienTrooper, parameterName=AT_EyeMetalic, sortPriority=100}
  ) {attributeSetTypes [
    #2
    #3
    #4
    #5
    #1
    #6
   ]}
  NamedRerouteDeclaration(None=@n49) {name=EyeMaterial, nodeColor=<rgba 0 0.5781 1 1>}
  NamedRerouteDeclaration(None=@n6.Result) {name=Preview, nodeColor=<rgba 1 0.8906 0 1>}
  NamedRerouteUsage()
  NamedRerouteUsage()
  NamedRerouteUsage()
  NamedRerouteUsage()
  TextureSampleParameter2D() {group=AlienTrooper, parameterName=AT_FXMask, sortPriority=5, texture=&T_BlacksRGB}
  TextureSampleParameter2D() {
   group=AlienTrooper
   parameterName=AT_Diffuse
   sortPriority=10
   texture=&T_M_MED_AlienTrooper_Horns_Head_D
  }
  VertexColor()
 ]}
