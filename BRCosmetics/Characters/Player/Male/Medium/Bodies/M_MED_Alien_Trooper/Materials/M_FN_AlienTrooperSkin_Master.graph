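# Master material for the Alien Trooper character's skin on the male medium body, built for skeletal meshes with morph targets and mesh deformers.
# It feeds the shared character PreFX function into the MF_AlienTrooperSkin function, then passes that result with PreFX outputs into the shared PostFX function.
# It uses material attributes, with the shading model chosen by expression between Default Lit and Subsurface Profile.
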Material /BRCosmetics/Characters/Player/Male/Medium/Bodies/M_MED_Alien_Trooper/Materials/M_FN_AlienTrooperSkin_Master
graph {
 properties {
  bUseMaterialAttributes=true
  bUsedWithMeshDeformer=true
  bUsedWithMorphTargets=true
  bUsedWithSkeletalMesh=true
  pixelDepthOffsetMode=PDOM_Legacy
  shadingModel=MSM_FromMaterialExpression
  usedShadingModels=Default Lit | Subsurface Profile
 }
 nodes [
  n1 = MaterialFunctionCall() {materialFunction=&BaseFN_CharacterFunction_PreFX}
  MaterialFunctionCall(
   AO=@n1.AO
   Distance Blend=@n1.AO
   Fresnel=@n1.AO
   Gmap=@n1.AO
   Material=MaterialFunctionCall!(Skin=@n1.Skin, Material=@n1.Skin) {materialFunction=&/MF_AlienTrooperSkin}.Result
  ) {materialFunction=&BaseFN_CharacterFunction_PostFX}
 ]
}
