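# Animation Blueprint for the Alien Trooper body cosmetic, copying the attached parent mesh's pose and correcting its shoulder pads.
# Weighted multi-bone constraints place both pad bases, then constraints and additive bone offsets blend in by pistol and rifle aim alphas.
# A sculpters Control Rig follows, then an emote Control Rig faded by the custom-emote alpha with interpolation.
# Each update sets GravityOverride through a deprecated state function.
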AnimBlueprint /BRCosmetics/Characters/Player/Male/Medium/Bodies/M_MED_Alien_Trooper/Meshes/M_MED_Alien_Trooper_AnimBP
target_skeleton=&/M_MED_Alien_Trooper_Skeleton
graphs [
 AnimGraph {nodes [
   n1 = Get PistolAimDownAlpha() -> (PistolAimDownAlpha:float)
   "Constraint - dyn_r_shoulderpad_base (PP)"!(ComponentPose, Alpha=@n1, ConstraintWeights_0=0.2, ConstraintWeights_1:float=0.8) -> (Pose) {
    alphaBoolBlend {blendOption=Linear}
    boneToModify {boneName=dyn_r_shoulderpad_base}
    constraintSetup [{targetBone {boneName=clavicle_r}}, {targetBone {boneName=upperarm_r}}]
    constraintWeights [1,1]
   }
   n2 = "Constraint - dyn_r_shoulderpad_base (PPPPPPP)"!(
    ComponentPose=Local To Component(LocalPose=Copy Pose From Mesh() {bUseAttachedParent=true})
    Alpha=1
    ConstraintWeights_0=0
    ConstraintWeights_1:float=0.5118
    ConstraintWeights_2:float=0
    ConstraintWeights_3:float=0.0316
    ConstraintWeights_4:float=0.0003
    ConstraintWeights_5:float=0.0353
    ConstraintWeights_6:float=0.421
   ) -> (Pose) {
    alphaBoolBlend {blendOption=Linear}
    boneToModify {boneName=dyn_r_shoulderpad_base}
    constraintSetup [
     {targetBone {boneName=upperarm_r}}
     {targetBone {boneName=upperarm_twist_01_r}}
     {targetBone {boneName=upperarm_twist_02_r}}
     {targetBone {boneName=spine_03}}
     {targetBone {boneName=spine_04}}
     {targetBone {boneName=spine_05}}
     {targetBone {boneName=clavicle_r}}
    ]
    constraintWeights [1,1,1,1,1,1,1]
   }
   n3 = "Constraint - dyn_l_shoulderpad_base (PPPPPPPPPP)"!(
    ComponentPose=@n2
    Alpha=1
    ConstraintWeights_0=0.375
    ConstraintWeights_1:float=0.006
    ConstraintWeights_2:float=0
    ConstraintWeights_3:float=0.0328
    ConstraintWeights_4:float=0.005
    ConstraintWeights_5:float=0.0391
    ConstraintWeights_6:float=0.5331
    ConstraintWeights_7:float=0.0013
    ConstraintWeights_8:float=0.0002
    ConstraintWeights_9:float=0.0075
   ) -> (Pose) {
    alphaBoolBlend {blendOption=Linear}
    boneToModify {boneName=dyn_l_shoulderpad_base}
    constraintSetup [
     {targetBone {boneName=clavicle_l}}
     {targetBone {boneName=upperarm_twist_02_l}}
     {targetBone {boneName=spine_02}}
     {targetBone {boneName=spine_03}}
     {targetBone {boneName=spine_04}}
     {targetBone {boneName=spine_05}}
     {targetBone {boneName=upperarm_twist_01_l}}
     {targetBone {boneName=neck_02}}
     {targetBone {boneName=neck_01}}
     {targetBone {boneName=upperarm_l}}
    ]
    constraintWeights [1,1,1,1,1,1,1,1,1,1]
   }
   n4 = Get PistolAimUpAlpha() -> (PistolAimUpAlpha:float)
   n5 = "Constraint - dyn_r_shoulderpad_base (PP)"(ComponentPose=@n3, Alpha=@n4, ConstraintWeights_0=0.2, ConstraintWeights_1=0.8) {
    alphaBoolBlend {blendOption=Linear}
    boneToModify {boneName=dyn_r_shoulderpad_base}
    constraintSetup [{targetBone {boneName=clavicle_r}}, {targetBone {boneName=upperarm_r}}]
    constraintWeights [1,1]
   }
   n6 = "Transform (Modify) Bone - Bone: dyn_r_shoulderpad_master"(ComponentPose=@n5, Alpha=@n4, Translation="0,0,1", Rotation="15,20,-10", Scale="1,1,1") {
    rotationMode=BMM_Additive
    rotationSpace=BCS_BoneSpace
    translationMode=BMM_Additive
    translationSpace=BCS_BoneSpace
    alphaBoolBlend {blendOption=Linear}
    boneToModify {boneName=dyn_r_shoulderpad_master}
    rotation=<rot 15 20 -10>
    translation {z=1}
   }
   n7 = "Constraint - dyn_l_shoulderpad_base (PP)"!(ComponentPose=@n6, Alpha=@n4, ConstraintWeights_0=0.2, ConstraintWeights_1:float=0.8) -> (Pose) {
    alphaBoolBlend {blendOption=Linear}
    boneToModify {boneName=dyn_l_shoulderpad_base}
    constraintSetup [{targetBone {boneName=clavicle_l}}, {targetBone {boneName=upperarm_l}}]
    constraintWeights [1,1]
   }
   n8 = "Transform (Modify) Bone - Bone: dyn_l_shoulderpad_master"(ComponentPose=@n7, Alpha=@n4, Translation="0,0,0", Rotation="5,0,0", Scale="1,1,1") {
    rotationMode=BMM_Additive
    rotationSpace=BCS_BoneSpace
    translationMode=BMM_Additive
    translationSpace=BCS_BoneSpace
    alphaBoolBlend {blendOption=Linear}
    boneToModify {boneName=dyn_l_shoulderpad_master}
    rotation {pitch=5}
   }
   n9 = "Transform (Modify) Bone - Bone: dyn_r_shoulderpad_base"(ComponentPose=@n8, Alpha=@n1, Translation="0,0,0", Rotation="0,0,-20", Scale="1,1,1") {
    rotationMode=BMM_Additive
    rotationSpace=BCS_BoneSpace
    translationMode=BMM_Additive
    translationSpace=BCS_BoneSpace
    alphaBoolBlend {blendOption=Linear}
    boneToModify {boneName=dyn_r_shoulderpad_base}
    rotation {roll=-20}
   }
   n10 = "Transform (Modify) Bone - Bone: dyn_r_shoulderpad_master"(ComponentPose=@n9, Alpha=@n1, Translation="0,0,0", Rotation="0,0,-15", Scale="1,1,1") {
    rotationMode=BMM_Additive
    rotationSpace=BCS_BoneSpace
    translationMode=BMM_Additive
    translationSpace=BCS_BoneSpace
    alphaBoolBlend {blendOption=Linear}
    boneToModify {boneName=dyn_r_shoulderpad_master}
    rotation {roll=-15}
   }
   n11 = "Transform (Modify) Bone - Bone: dyn_r_shoulderpad_base"(ComponentPose=@n10, Alpha=Get PistolAimMidAlpha() -> (PistolAimMidAlpha:float), Translation="0,0,0", Rotation="0,0,-15", Scale="1,1,1") {
    rotationMode=BMM_Additive
    rotationSpace=BCS_BoneSpace
    translationMode=BMM_Additive
    translationSpace=BCS_BoneSpace
    alphaBoolBlend {blendOption=Linear}
    boneToModify {boneName=dyn_r_shoulderpad_base}
    rotation {roll=-15}
   }
   n12 = Get RifleAimUpAlpha() -> (RifleAimUpAlpha:float)
   n13 = "Transform (Modify) Bone - Bone: dyn_r_shoulderpad_base"(ComponentPose=@n11, Alpha=@n12, Translation="0,0,0", Rotation="0,0,-10", Scale="1,1,1") {
    rotationMode=BMM_Additive
    rotationSpace=BCS_BoneSpace
    translationMode=BMM_Additive
    translationSpace=BCS_BoneSpace
    alphaBoolBlend {blendOption=Linear}
    boneToModify {boneName=dyn_r_shoulderpad_base}
    rotation {roll=-10}
   }
   n14 = "Transform (Modify) Bone - Bone: dyn_l_shoulderpad_master"(ComponentPose=@n13, Alpha=@n12, Translation="0,0,0", Rotation="0,0,-30", Scale="1,1,1") {
    rotationMode=BMM_Additive
    rotationSpace=BCS_BoneSpace
    translationMode=BMM_Additive
    translationSpace=BCS_BoneSpace
    alphaBoolBlend {blendOption=Linear}
    boneToModify {boneName=dyn_l_shoulderpad_master}
    rotation {roll=-30}
   }
   n15 = "Transform (Modify) Bone - Bone: dyn_l_shoulderpad_base"(ComponentPose=@n14, Alpha=@n12, Translation="0,0,2", Rotation="5,15,0", Scale="1,1,1") {
    rotationMode=BMM_Additive
    rotationSpace=BCS_BoneSpace
    translationMode=BMM_Additive
    translationSpace=BCS_BoneSpace
    alphaBoolBlend {blendOption=Linear}
    boneToModify {boneName=dyn_l_shoulderpad_base}
    rotation {pitch=5, yaw=15}
    translation {z=2}
   }
   n16 = "Transform (Modify) Bone - Bone: dyn_l_shoulderpad_base"(ComponentPose=@n15, Alpha=Get RifleAimDownAlpha() -> (RifleAimDownAlpha:float), Translation="0,0,0", Rotation="0,0,5", Scale="1,1,1") {
    rotationMode=BMM_Additive
    rotationSpace=BCS_BoneSpace
    translationMode=BMM_Additive
    translationSpace=BCS_BoneSpace
    alphaBoolBlend {blendOption=Linear}
    boneToModify {boneName=dyn_l_shoulderpad_base}
    rotation {roll=5}
   }
   n17 = Control Rig(Source=Component To Local(ComponentPose=@n16)) {
    alphaBoolBlend {blendOption=Linear}
    controlRigAssetReference {blueprintRigClass=&/M_MED_Alien_Trooper_CtrlRig_Sculpters.M_MED_Alien_Trooper_CtrlRig_Sculpters_C}
   }
   n18 = Control Rig(Source=@n17, Alpha=Get IsCustomEmotePlayingAlpha() -> (IsCustomEmotePlayingAlpha:float)) {
    alphaBoolBlend {blendOption=Linear}
    alphaScaleBiasClamp {bInterpResult=true, interpSpeedDecreasing=20, interpSpeedIncreasing=20}
    controlRigAssetReference {blueprintRigClass=&/M_MED_Alien_Trooper_CtrlRig_emote.M_MED_Alien_Trooper_CtrlRig_emote_C}
   }
   Output Pose(Result=@n18) {layerGroup=None}
  ]}
 EventGraph {nodes [
   Event BlueprintUpdateAnimation()
   > n1 = Gravity Override State Function DEPRECATED(CharacterPartAnimInstance=Self-Reference(), GameTimeInSeconds=GetGameTimeInSeconds(), GravityOverrideParams=Get Gravity Override Params())
   > Set GravityOverride(GravityOverride=@n1)
  ]}
]
variables [Gravity Override Params {type=GravityOverrideParamsStruct}]
